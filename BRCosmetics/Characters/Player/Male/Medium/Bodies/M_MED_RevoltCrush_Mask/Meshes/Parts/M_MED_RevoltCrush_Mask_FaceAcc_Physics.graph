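# Physics asset for the male medium RevoltCrush Mask face accessory, defining secondary motion for its hair.
# Small base spheres link to the head and root chains of mostly capsule bodies, with tight swing limits, often widening toward the tips, and locked twist.
# Head, neck, upper spine, clavicle and upper-arm capsules supply body volumes around the hair.
# Both lower-side "a" chains branch: their 03 bodies constrain to 01, not 02.
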PhysicsAsset /BRCosmetics/Characters/Player/Male/Medium/Bodies/M_MED_RevoltCrush_Mask/Meshes/Parts/M_MED_RevoltCrush_Mask_FaceAcc_Physics
bodies {
 base_back_hair [{shapeType=Sphere, radius=0.505, shapeName=base_back_hair_sphere}]
 base_back_hair_l [{shapeType=Sphere, radius=0.505, shapeName=base_back_hair_l_sphere}]
 base_back_hair_r [{shapeType=Sphere, radius=0.505, shapeName=base_back_hair_r_sphere}]
 base_back_lower_hair [{shapeType=Sphere, radius=0.505, shapeName=base_back_lower_hair_sphere}]
 base_back_mid_hair [{shapeType=Sphere, radius=0.505, shapeName=base_back_mid_hair_sphere}]
 base_back_mid_hair_r [{shapeType=Sphere, radius=0.505, shapeName=base_back_mid_hair_r_sphere}]
 base_side_hair_l_a [{shapeType=Sphere, radius=0.505, shapeName=base_side_hair_l_a_sphere}]
 base_side_hair_l_b [{shapeType=Sphere, radius=0.505, shapeName=base_side_hair_l_b_sphere}]
 base_side_hair_l_c [{shapeType=Sphere, radius=0.505, shapeName=base_side_hair_l_c_sphere}]
 base_side_hair_r_a [{shapeType=Sphere, radius=0.505, shapeName=base_side_hair_r_a_sphere}]
 base_side_hair_r_b [{shapeType=Sphere, radius=0.505, shapeName=base_side_hair_r_b_sphere}]
 base_side_lower_hair_l_a [{shapeType=Sphere, radius=0.505, shapeName=base_side_lower_hair_l_a_sphere}]
 base_side_lower_hair_l_b [{shapeType=Sphere, radius=0.505, shapeName=base_side_lower_hair_l_b_sphere}]
 base_side_lower_hair_r_a [{shapeType=Sphere, radius=0.505, shapeName=base_side_lower_hair_r_a_sphere}]
 base_side_lower_hair_r_b [{shapeType=Sphere, radius=0.505, shapeName=base_side_lower_hair_r_b_sphere}]
 base_side_mid_hair_l [{shapeType=Sphere, radius=0.505, shapeName=base_side_mid_hair_l_sphere}]
 clavicle_l [{shapeType=Capsule, length=11, radius=9, shapeName=clavicle_l_capsule, center=<5.2471 0 -2.6613>, rotation=<rot -0.0001 -90 -89.9583>}]
 clavicle_r [{shapeType=Capsule, length=11, radius=9, shapeName=clavicle_r_capsule, center=<-5.0075 0 2.4362>, rotation=<rot -0.0001 -90 -89.9583>}]
 dyn_back_hair_01 [
  {shapeType=Sphere, radius=1, center=<10 0 -2.6303>}
  {shapeType=Capsule, length=3, radius=2, shapeName=dyn_back_hair_01_capsule, center=<5 0 0>, rotation=<rot 90 0 0>}
 ]
 dyn_back_hair_02 [{shapeType=Capsule, length=3, radius=2, shapeName=dyn_back_hair_02_capsule, center=<5 0 0>, rotation=<rot 90 0 0>}]
 dyn_back_hair_03 [{shapeType=Capsule, length=3, radius=2, shapeName=dyn_back_hair_03_capsule, center=<5 0 0>, rotation=<rot 90 0 0>}]
 dyn_back_hair_l_01 [
  {shapeType=Sphere, radius=1, center=<10 0 -2.6303>}
  {shapeType=Capsule, length=3, radius=2, shapeName=dyn_back_hair_l_01_capsule, center=<5 0 0>, rotation=<rot 90 0 0>}
 ]
 dyn_back_hair_l_02 [{shapeType=Capsule, length=3, radius=2, shapeName=dyn_back_hair_l_02_capsule, center=<5 0 0>, rotation=<rot 90 0 0>}]
 dyn_back_hair_l_03 [{shapeType=Capsule, length=3, radius=2, shapeName=dyn_back_hair_l_03_capsule, center=<5 0 0>, rotation=<rot 90 0 0>}]
 dyn_back_hair_r_01 [
  {shapeType=Sphere, radius=1, center=<10 0 -2.6303>}
  {shapeType=Capsule, length=3, radius=2, shapeName=dyn_back_hair_r_01_capsule, center=<5 0 0>, rotation=<rot 90 0 0>}
 ]
 dyn_back_hair_r_02 [{shapeType=Capsule, length=3, radius=2, shapeName=dyn_back_hair_r_02_capsule, center=<5 0 0>, rotation=<rot 90 0 0>}]
 dyn_back_hair_r_03 [{shapeType=Capsule, length=3, radius=2, shapeName=dyn_back_hair_r_03_capsule, center=<5 0 0>, rotation=<rot 90 0 0>}]
 dyn_back_lower_hair_01 [{shapeType=Capsule, length=3, radius=3, shapeName=dyn_back_lower_hair_01_capsule, center=<5 0 0>, rotation=<rot 90 0 0>}]
 dyn_back_lower_hair_02 [{shapeType=Capsule, length=3, radius=3, shapeName=dyn_back_lower_hair_02_capsule, center=<5 0 0>, rotation=<rot 90 0 0>}]
 dyn_back_mid_hair_l_01 [{shapeType=Capsule, length=3, radius=2, shapeName=dyn_back_mid_hair_l_01_capsule, center=<5 0 0>, rotation=<rot 90 0 0>}]
 dyn_back_mid_hair_l_02 [{shapeType=Capsule, length=3, radius=2, shapeName=dyn_back_mid_hair_l_02_capsule, center=<5 0 0>, rotation=<rot 90 0 0>}]
 dyn_back_mid_hair_r_01 [{shapeType=Capsule, length=3, radius=2, shapeName=dyn_back_mid_hair_r_01_capsule, center=<5 0 0>, rotation=<rot 90 0 0>}]
 dyn_back_mid_hair_r_02 [{shapeType=Capsule, length=3, radius=2, shapeName=dyn_back_mid_hair_r_02_capsule, center=<5 0 0>, rotation=<rot 90 0 0>}]
 dyn_side_hair_l_01_a [
  {shapeType=Sphere, radius=1, center=<10 0 -2.6303>}
  {shapeType=Capsule, length=3, radius=2, shapeName=dyn_side_hair_l_01_a_capsule, center=<5 0 0>, rotation=<rot 90 0 0>}
 ]
 dyn_side_hair_l_01_b [
  {shapeType=Sphere, radius=1, center=<10 0 -2.6303>}
  {shapeType=Capsule, length=3, radius=2, shapeName=dyn_side_hair_l_01_b_capsule, center=<5 0 0>, rotation=<rot 90 0 0>}
 ]
 dyn_side_hair_l_01_c [
  {shapeType=Sphere, radius=1, center=<10 0 -2.6303>}
  {shapeType=Capsule, length=3, radius=2, shapeName=dyn_side_hair_l_01_c_capsule, center=<5 0 0>, rotation=<rot 90 0 0>}
 ]
 dyn_side_hair_l_02_a [{shapeType=Capsule, length=3, radius=2, shapeName=dyn_side_hair_l_02_a_capsule, center=<5 0 0>, rotation=<rot 90 0 0>}]
 dyn_side_hair_l_02_b [
  {shapeType=Sphere, radius=1, center=<10 0 -2.6303>}
  {shapeType=Capsule, length=3, radius=2, shapeName=dyn_side_hair_l_02_b_capsule, center=<5 0 0>, rotation=<rot 90 0 0>}
 ]
 dyn_side_hair_l_02_c [{shapeType=Capsule, length=3, radius=2, shapeName=dyn_side_hair_l_02_c_capsule, center=<5 0 0>, rotation=<rot 90 0 0>}]
 dyn_side_hair_l_03_a [{shapeType=Capsule, length=3, radius=2, shapeName=dyn_side_hair_l_03_a_capsule, center=<5 0 0>, rotation=<rot 90 0 0>}]
 dyn_side_hair_l_03_b [{shapeType=Capsule, length=3, radius=2, shapeName=dyn_side_hair_l_03_b_capsule, center=<5 0 0>, rotation=<rot 90 0 0>}]
 dyn_side_hair_l_03_c [{shapeType=Capsule, length=3, radius=2, shapeName=dyn_side_hair_l_03_c_capsule, center=<5 0 0>, rotation=<rot 90 0 0>}]
 dyn_side_hair_r_01_a [{shapeType=Capsule, length=3, radius=2, shapeName=dyn_side_hair_r_01_a_capsule, center=<5 0 0>, rotation=<rot 90 0 0>}]
 dyn_side_hair_r_01_b [{shapeType=Capsule, length=3, radius=2, shapeName=dyn_side_hair_r_01_b_capsule, center=<5 0 0>, rotation=<rot 90 0 0>}]
 dyn_side_hair_r_02_a [{shapeType=Capsule, length=3.5, radius=1.5, shapeName=dyn_side_hair_r_02_a_capsule, center=<5 0 0>, rotation=<rot 90 0 0>}]
 dyn_side_hair_r_02_b [{shapeType=Capsule, length=3.5, radius=1.5, shapeName=dyn_side_hair_r_02_b_capsule, center=<5 0 0>, rotation=<rot 90 0 0>}]
 dyn_side_hair_r_03_a [{shapeType=Capsule, length=4, radius=1.5, shapeName=dyn_side_hair_r_03_a_capsule, center=<5 0 0>, rotation=<rot 90 0 0>}]
 dyn_side_hair_r_03_b [{shapeType=Capsule, length=4.5, radius=1, shapeName=dyn_side_hair_r_03_b_capsule, center=<5 0 0>, rotation=<rot 90 0 0>}]
 dyn_side_lower_hair_l_01_a [{shapeType=Capsule, length=2.5, radius=2.5, shapeName=dyn_side_lower_hair_l_01_a_capsule, center=<5 0 0>, rotation=<rot 90 0 0>}]
 dyn_side_lower_hair_l_01_b [{shapeType=Capsule, length=2.5, radius=3.5, shapeName=dyn_side_lower_hair_l_01_b_capsule, center=<5 0 0>, rotation=<rot 90 0 0>}]
 dyn_side_lower_hair_l_02_a [{shapeType=Capsule, length=3, radius=2, shapeName=dyn_side_lower_hair_l_02_a_capsule, center=<5 0 0>, rotation=<rot 90 0 0>}]
 dyn_side_lower_hair_l_02_b [{shapeType=Capsule, length=2.5, radius=3.5, shapeName=dyn_side_lower_hair_l_02_b_capsule, center=<5 0 0>, rotation=<rot 90 0 0>}]
 dyn_side_lower_hair_l_03_a [{shapeType=Capsule, length=2.5, radius=2, shapeName=dyn_side_lower_hair_l_03_a_capsule, center=<5 0 0>, rotation=<rot 90 0 0>}]
 dyn_side_lower_hair_r_01_a [{shapeType=Capsule, length=2.5, radius=2.5, shapeName=dyn_side_lower_hair_r_01_a_capsule, center=<5 0 0>, rotation=<rot 90 0 0>}]
 dyn_side_lower_hair_r_01_b [{shapeType=Capsule, length=2, radius=2, shapeName=dyn_side_lower_hair_r_01_b_capsule, center=<4.3785 0 0>, rotation=<rot 90 0 0>}]
 dyn_side_lower_hair_r_02_a [{shapeType=Capsule, length=2.5, radius=2.5, shapeName=dyn_side_lower_hair_r_02_a_capsule, center=<5 0 0>, rotation=<rot 90 0 0>}]
 dyn_side_lower_hair_r_02_b [{shapeType=Capsule, length=1, radius=2, shapeName=dyn_side_lower_hair_r_02_b_capsule, center=<2.7591 0 0>, rotation=<rot 90 0 0>}]
 dyn_side_lower_hair_r_03_a [{shapeType=Capsule, length=2.5, radius=2.5, shapeName=dyn_side_lower_hair_r_03_a_capsule, center=<5 0 0>, rotation=<rot 90 0 0>}]
 dyn_side_mid_hair_l_01 [{shapeType=Capsule, length=3, radius=2, shapeName=dyn_side_mid_hair_l_01_capsule, center=<5 0 0>, rotation=<rot 90 0 0>}]
 dyn_side_mid_hair_l_02 [{shapeType=Capsule, length=3.5, radius=2, shapeName=dyn_side_mid_hair_l_02_capsule, center=<5 0 0>, rotation=<rot 90 0 0>}]
 head [{shapeType=Capsule, length=9.6707, radius=6.5, center=<1.1689 -0.1875 -0.3085>, rotation=<rot 0 -83.9184 -89.4323>}]
 neck_01 [{shapeType=Capsule, length=3.3738, radius=6.6768, shapeName=neck_01_capsule, center=<2.8784 -1.4244 0>, rotation=<rot 90 0 0>}]
 spine_05 [
  {shapeType=Capsule, length=6.7252, radius=6.7602, center=<8.8061 4.232 -0.1685>, rotation=<rot 0 6.7171 90>}
  {shapeType=Capsule, length=8.5, radius=11, shapeName=spine_05_capsule, center=<3.1487 0.6839 0>}
 ]
 upperarm_l [{shapeType=Capsule, length=37, radius=6.5, shapeName=upperarm_l_capsule, center=<19.3226 0 1.634>, rotation=<rot 85.1662 -179.9998 -179.9998>}]
 upperarm_r [{
   shapeType=Capsule
   length=38.5
   radius=5
   shapeName=upperarm_r_capsule
   center=<-20.3573 -0.0022 -1.5988>
   rotation=<rot 85.5094 -179.8947 -179.9013>
  }]
}
constraints [
 {bone1Name=head, bone2Name=neck_01, swing1LimitDegrees=45, swing2LimitDegrees=45, twistLimitDegrees=45}
 {bone1Name=dyn_back_hair_01, bone2Name=base_back_hair, swing1LimitDegrees=5, swing2LimitDegrees=10, twistLimitDegrees=1, twistMotion=Locked}
 {
  bone1Name=dyn_back_hair_02
  bone2Name=dyn_back_hair_01
  swing1LimitDegrees=10
  swing2LimitDegrees=15
  twistLimitDegrees=1
  twistMotion=Locked
 }
 {
  bone1Name=dyn_back_hair_03
  bone2Name=dyn_back_hair_02
  swing1LimitDegrees=15
  swing2LimitDegrees=25
  twistLimitDegrees=1
  twistMotion=Locked
 }
 {
  bone1Name=dyn_back_hair_l_01
  bone2Name=base_back_hair_l
  swing1LimitDegrees=3
  swing2LimitDegrees=10
  twistLimitDegrees=1
  twistMotion=Locked
 }
 {
  bone1Name=dyn_back_hair_l_02
  bone2Name=dyn_back_hair_l_01
  swing1LimitDegrees=10
  swing2LimitDegrees=20
  twistLimitDegrees=1
  twistMotion=Locked
 }
 {
  bone1Name=dyn_back_hair_l_03
  bone2Name=dyn_back_hair_l_02
  swing1LimitDegrees=15
  swing2LimitDegrees=25
  twistLimitDegrees=1
  twistMotion=Locked
 }
 {
  bone1Name=dyn_back_hair_r_01
  bone2Name=base_back_hair_r
  swing1LimitDegrees=15
  swing2LimitDegrees=10
  twistLimitDegrees=1
  twistMotion=Locked
 }
 {
  bone1Name=dyn_back_hair_r_02
  bone2Name=dyn_back_hair_r_01
  swing1LimitDegrees=15
  swing2LimitDegrees=15
  twistLimitDegrees=1
  twistMotion=Locked
 }
 {
  bone1Name=dyn_back_hair_r_03
  bone2Name=dyn_back_hair_r_02
  swing1LimitDegrees=15
  swing2LimitDegrees=25
  twistLimitDegrees=1
  twistMotion=Locked
 }
 {
  bone1Name=dyn_back_lower_hair_01
  bone2Name=base_back_lower_hair
  swing1LimitDegrees=15
  swing2LimitDegrees=30
  twistLimitDegrees=1
  twistMotion=Locked
 }
 {
  bone1Name=dyn_back_lower_hair_02
  bone2Name=dyn_back_lower_hair_01
  swing1LimitDegrees=15
  swing2LimitDegrees=30
  twistLimitDegrees=1
  twistMotion=Locked
 }
 {
  bone1Name=dyn_back_mid_hair_l_01
  bone2Name=base_back_mid_hair
  swing1LimitDegrees=15
  swing2LimitDegrees=20
  twistLimitDegrees=1
  twistMotion=Locked
 }
 {
  bone1Name=dyn_back_mid_hair_l_02
  bone2Name=dyn_back_mid_hair_l_01
  swing1LimitDegrees=15
  swing2LimitDegrees=30
  twistLimitDegrees=1
  twistMotion=Locked
 }
 {
  bone1Name=dyn_back_mid_hair_r_01
  bone2Name=base_back_mid_hair_r
  swing1LimitDegrees=15
  swing2LimitDegrees=20
  twistLimitDegrees=1
  twistMotion=Locked
 }
 {
  bone1Name=dyn_back_mid_hair_r_02
  bone2Name=dyn_back_mid_hair_r_01
  swing1LimitDegrees=15
  swing2LimitDegrees=30
  twistLimitDegrees=1
  twistMotion=Locked
 }
 {
  bone1Name=dyn_side_hair_l_01_a
  bone2Name=base_side_hair_l_a
  swing1LimitDegrees=15
  swing2LimitDegrees=10
  twistLimitDegrees=1
  twistMotion=Locked
 }
 {
  bone1Name=dyn_side_hair_l_02_a
  bone2Name=dyn_side_hair_l_01_a
  swing1LimitDegrees=15
  swing2LimitDegrees=15
  twistLimitDegrees=1
  twistMotion=Locked
 }
 {
  bone1Name=dyn_side_hair_l_03_a
  bone2Name=dyn_side_hair_l_02_a
  swing1LimitDegrees=15
  swing2LimitDegrees=20
  twistLimitDegrees=1
  twistMotion=Locked
 }
 {
  bone1Name=dyn_side_hair_l_01_b
  bone2Name=base_side_hair_l_b
  swing1LimitDegrees=15
  swing2LimitDegrees=10
  twistLimitDegrees=1
  twistMotion=Locked
 }
 {
  bone1Name=dyn_side_hair_l_02_b
  bone2Name=dyn_side_hair_l_01_b
  swing1LimitDegrees=15
  swing2LimitDegrees=15
  twistLimitDegrees=1
  twistMotion=Locked
 }
 {
  bone1Name=dyn_side_hair_l_03_b
  bone2Name=dyn_side_hair_l_02_b
  swing1LimitDegrees=15
  swing2LimitDegrees=20
  twistLimitDegrees=1
  twistMotion=Locked
 }
 {
  bone1Name=dyn_side_hair_l_01_c
  bone2Name=base_side_hair_l_c
  swing1LimitDegrees=15
  swing2LimitDegrees=10
  twistLimitDegrees=1
  twistMotion=Locked
 }
 {
  bone1Name=dyn_side_hair_l_02_c
  bone2Name=dyn_side_hair_l_01_c
  swing1LimitDegrees=15
  swing2LimitDegrees=15
  twistLimitDegrees=1
  twistMotion=Locked
 }
 {
  bone1Name=dyn_side_hair_l_03_c
  bone2Name=dyn_side_hair_l_02_c
  swing1LimitDegrees=15
  swing2LimitDegrees=20
  twistLimitDegrees=1
  twistMotion=Locked
 }
 {
  bone1Name=dyn_side_hair_r_01_a
  bone2Name=base_side_hair_r_a
  swing1LimitDegrees=5
  swing2LimitDegrees=10
  twistLimitDegrees=1
  twistMotion=Locked
 }
 {
  bone1Name=dyn_side_hair_r_02_a
  bone2Name=dyn_side_hair_r_01_a
  swing1LimitDegrees=10
  swing2LimitDegrees=15
  twistLimitDegrees=1
  twistMotion=Locked
 }
 {
  bone1Name=dyn_side_hair_r_03_a
  bone2Name=dyn_side_hair_r_02_a
  swing1LimitDegrees=15
  swing2LimitDegrees=20
  twistLimitDegrees=1
  twistMotion=Locked
 }
 {
  bone1Name=dyn_side_hair_r_01_b
  bone2Name=base_side_hair_r_b
  swing1LimitDegrees=15
  swing2LimitDegrees=10
  twistLimitDegrees=1
  twistMotion=Locked
 }
 {
  bone1Name=dyn_side_hair_r_02_b
  bone2Name=dyn_side_hair_r_01_b
  swing1LimitDegrees=15
  swing2LimitDegrees=15
  twistLimitDegrees=1
  twistMotion=Locked
 }
 {
  bone1Name=dyn_side_hair_r_03_b
  bone2Name=dyn_side_hair_r_02_b
  swing1LimitDegrees=15
  swing2LimitDegrees=20
  twistLimitDegrees=1
  twistMotion=Locked
 }
 {
  bone1Name=dyn_side_lower_hair_l_01_a
  bone2Name=base_side_lower_hair_l_a
  swing1LimitDegrees=8
  swing2LimitDegrees=30
  twistLimitDegrees=1
  twistMotion=Locked
 }
 {
  bone1Name=dyn_side_lower_hair_l_02_a
  bone2Name=dyn_side_lower_hair_l_01_a
  swing1LimitDegrees=8
  swing2LimitDegrees=30
  twistLimitDegrees=1
  twistMotion=Locked
 }
 {
  bone1Name=dyn_side_lower_hair_l_03_a
  bone2Name=dyn_side_lower_hair_l_01_a
  swing1LimitDegrees=8
  swing2LimitDegrees=30
  twistLimitDegrees=1
  twistMotion=Locked
 }
 {
  bone1Name=dyn_side_lower_hair_l_01_b
  bone2Name=base_side_lower_hair_l_b
  swing1LimitDegrees=8
  swing2LimitDegrees=20
  twistLimitDegrees=1
  twistMotion=Locked
 }
 {
  bone1Name=dyn_side_lower_hair_l_02_b
  bone2Name=dyn_side_lower_hair_l_01_b
  swing1LimitDegrees=8
  swing2LimitDegrees=25
  twistLimitDegrees=1
  twistMotion=Locked
 }
 {
  bone1Name=dyn_side_lower_hair_r_01_a
  bone2Name=base_side_lower_hair_r_a
  swing1LimitDegrees=8
  swing2LimitDegrees=20
  twistLimitDegrees=1
  twistMotion=Locked
 }
 {
  bone1Name=dyn_side_lower_hair_r_02_a
  bone2Name=dyn_side_lower_hair_r_01_a
  swing1LimitDegrees=8
  swing2LimitDegrees=20
  twistLimitDegrees=1
  twistMotion=Locked
 }
 {
  bone1Name=dyn_side_lower_hair_r_03_a
  bone2Name=dyn_side_lower_hair_r_01_a
  swing1LimitDegrees=8
  swing2LimitDegrees=20
  twistLimitDegrees=1
  twistMotion=Locked
 }
 {
  bone1Name=dyn_side_lower_hair_r_01_b
  bone2Name=base_side_lower_hair_r_b
  swing1LimitDegrees=8
  swing2LimitDegrees=25
  twistLimitDegrees=1
  twistMotion=Locked
 }
 {
  bone1Name=dyn_side_lower_hair_r_02_b
  bone2Name=dyn_side_lower_hair_r_01_b
  swing1LimitDegrees=8
  swing2LimitDegrees=25
  twistLimitDegrees=1
  twistMotion=Locked
 }
 {
  bone1Name=dyn_side_mid_hair_l_01
  bone2Name=base_side_mid_hair_l
  swing1LimitDegrees=15
  swing2LimitDegrees=15
  twistLimitDegrees=1
  twistMotion=Locked
 }
 {
  bone1Name=dyn_side_mid_hair_l_02
  bone2Name=dyn_side_mid_hair_l_01
  swing1LimitDegrees=15
  swing2LimitDegrees=20
  twistLimitDegrees=1
  twistMotion=Locked
 }
 {
  bone1Name=clavicle_r
  bone2Name=spine_05
  swing1LimitDegrees=45
  swing1Motion=Free
  swing2LimitDegrees=45
  swing2Motion=Free
  twistLimitDegrees=45
  twistMotion=Free
 }
 {
  bone1Name=clavicle_l
  bone2Name=spine_05
  swing1LimitDegrees=45
  swing1Motion=Free
  swing2LimitDegrees=45
  swing2Motion=Free
  twistLimitDegrees=45
  twistMotion=Free
 }
 {
  bone1Name=neck_01
  bone2Name=spine_05
  swing1LimitDegrees=45
  swing1Motion=Free
  swing2LimitDegrees=45
  swing2Motion=Free
  twistLimitDegrees=45
  twistMotion=Free
 }
 {
  bone1Name=base_side_mid_hair_l
  bone2Name=head
  swing1LimitDegrees=45
  swing1Motion=Free
  swing2LimitDegrees=45
  swing2Motion=Free
  twistLimitDegrees=45
  twistMotion=Free
 }
 {
  bone1Name=base_back_hair
  bone2Name=head
  swing1LimitDegrees=45
  swing1Motion=Free
  swing2LimitDegrees=45
  swing2Motion=Free
  twistLimitDegrees=45
  twistMotion=Free
 }
 {
  bone1Name=base_back_hair_l
  bone2Name=head
  swing1LimitDegrees=45
  swing1Motion=Free
  swing2LimitDegrees=45
  swing2Motion=Free
  twistLimitDegrees=45
  twistMotion=Free
 }
 {
  bone1Name=base_back_hair_r
  bone2Name=head
  swing1LimitDegrees=45
  swing1Motion=Free
  swing2LimitDegrees=45
  swing2Motion=Free
  twistLimitDegrees=45
  twistMotion=Free
 }
 {
  bone1Name=base_back_lower_hair
  bone2Name=head
  swing1LimitDegrees=45
  swing1Motion=Free
  swing2LimitDegrees=45
  swing2Motion=Free
  twistLimitDegrees=45
  twistMotion=Free
 }
 {
  bone1Name=base_back_mid_hair
  bone2Name=head
  swing1LimitDegrees=45
  swing1Motion=Free
  swing2LimitDegrees=45
  swing2Motion=Free
  twistLimitDegrees=45
  twistMotion=Free
 }
 {
  bone1Name=base_back_mid_hair_r
  bone2Name=head
  swing1LimitDegrees=45
  swing1Motion=Free
  swing2LimitDegrees=45
  swing2Motion=Free
  twistLimitDegrees=45
  twistMotion=Free
 }
 {
  bone1Name=base_side_hair_l_a
  bone2Name=head
  swing1LimitDegrees=45
  swing1Motion=Free
  swing2LimitDegrees=45
  swing2Motion=Free
  twistLimitDegrees=45
  twistMotion=Free
 }
 {
  bone1Name=base_side_hair_l_b
  bone2Name=head
  swing1LimitDegrees=45
  swing1Motion=Free
  swing2LimitDegrees=45
  swing2Motion=Free
  twistLimitDegrees=45
  twistMotion=Free
 }
 {
  bone1Name=base_side_hair_l_c
  bone2Name=head
  swing1LimitDegrees=45
  swing1Motion=Free
  swing2LimitDegrees=45
  swing2Motion=Free
  twistLimitDegrees=45
  twistMotion=Free
 }
 {
  bone1Name=base_side_hair_r_a
  bone2Name=head
  swing1LimitDegrees=45
  swing1Motion=Free
  swing2LimitDegrees=45
  swing2Motion=Free
  twistLimitDegrees=45
  twistMotion=Free
 }
 {
  bone1Name=base_side_hair_r_b
  bone2Name=head
  swing1LimitDegrees=45
  swing1Motion=Free
  swing2LimitDegrees=45
  swing2Motion=Free
  twistLimitDegrees=45
  twistMotion=Free
 }
 {
  bone1Name=base_side_lower_hair_l_a
  bone2Name=head
  swing1LimitDegrees=45
  swing1Motion=Free
  swing2LimitDegrees=45
  swing2Motion=Free
  twistLimitDegrees=45
  twistMotion=Free
 }
 {
  bone1Name=base_side_lower_hair_l_b
  bone2Name=head
  swing1LimitDegrees=45
  swing1Motion=Free
  swing2LimitDegrees=45
  swing2Motion=Free
  twistLimitDegrees=45
  twistMotion=Free
 }
 {
  bone1Name=base_side_lower_hair_r_a
  bone2Name=head
  swing1LimitDegrees=45
  swing1Motion=Free
  swing2LimitDegrees=45
  swing2Motion=Free
  twistLimitDegrees=45
  twistMotion=Free
 }
 {
  bone1Name=base_side_lower_hair_r_b
  bone2Name=head
  swing1LimitDegrees=45
  swing1Motion=Free
  swing2LimitDegrees=45
  swing2Motion=Free
  twistLimitDegrees=45
  twistMotion=Free
 }
 {
  bone1Name=upperarm_l
  bone2Name=clavicle_l
  swing1LimitDegrees=45
  swing1Motion=Free
  swing2LimitDegrees=45
  swing2Motion=Free
  twistLimitDegrees=45
  twistMotion=Free
 }
 {
  bone1Name=upperarm_r
  bone2Name=clavicle_r
  swing1LimitDegrees=45
  swing1Motion=Free
  swing2LimitDegrees=45
  swing2Motion=Free
  twistLimitDegrees=45
  twistMotion=Free
 }
]
