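# Animation Blueprint for the face accessory of the RevoltCrush MaskBlank male cosmetic, posing and simulating its mask and hair bones.
# It copies the attached parent's pose, offsets Mask, re-poses hair bones, runs RigidBody in spine_05 space using the non-Blank Mask physics asset, then rescales hair with Scale vector.
# The event graph sets that vector, plus gravity override, LOD2 and head-yaw values the AnimGraph never reads.
# A second neck_02 RigidBody node is unconnected.
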AnimBlueprint /BRCosmetics/Characters/Player/Male/Medium/Bodies/M_MED_RevoltCrush_MaskBlank/Meshes/Parts/M_MED_RevoltCrush_MaskBlank_FaceAcc_AnimBP
target_skeleton=&/M_MED_RevoltCrush_MaskBlank_FaceAcc_Skeleton
graphs [
 AnimGraph {nodes [
   RigidBody!(ComponentPose, Alpha, OverrideWorldGravity:Vector="0,0,0", ExternalForce) -> (Pose) {
    bClampLinearTranslationLimitToRefPose=true
    bOverrideWorldGravity=true
    overridePhysicsAsset=/BRCosmetics/Characters/Player/Male/Medium/Bodies/M_MED_RevoltCrush_Mask/Meshes/Parts/M_MED_RevoltCrush_Mask_FaceAcc_Physics
    simulationSpace=BaseBoneSpace
    alphaBoolBlend {blendOption=Linear}
    baseBoneRef {boneName=neck_02}
    componentAppliedLinearAccClamp=<3000 3000 3000>
    overrideWorldGravity {z=-980}
    simSpaceSettings {maxAngularAcceleration=10, maxAngularVelocity=3, maxLinearAcceleration=10, maxLinearVelocity=10, velocityScaleZ=0, worldAlpha=1}
   }
   n1 = "Transform (Modify) Bone - Bone: Mask"!(
    ComponentPose=Local To Component(LocalPose=Copy Pose From Mesh() {bUseAttachedParent=true})
    bAlphaBoolEnabled:bool=True
    Translation="0,1.314,0"
    Rotation="0,0,0"
    Scale="1,1,1"
   ) -> (Pose) {
    alphaInputType=Bool
    rotationMode=BMM_Additive
    scaleMode=BMM_Replace
    translationMode=BMM_Additive
    translationSpace=BCS_BoneSpace
    alphaBoolBlend {blendOption=Linear}
    boneToModify {boneName=Mask}
    translation {y=1.314}
   }
   n2 = "Transform (Modify) Bone - Bone: base_back_hair"(ComponentPose=@n1, Alpha=1, Translation="0,0,0", Rotation="50,0,-0.004", Scale="1,1,1") {
    rotationMode=BMM_Additive
    rotationSpace=BCS_BoneSpace
    scaleMode=BMM_Replace
    scaleSpace=BCS_BoneSpace
    translationMode=BMM_Additive
    translationSpace=BCS_BoneSpace
    alphaBoolBlend {blendOption=Linear}
    boneToModify {boneName=base_back_hair}
    rotation {pitch=50, roll=-0.004}
   }
   n3 = "Transform (Modify) Bone - Bone: base_back_hair_l"(ComponentPose=@n2, Alpha=1, Translation="0,0,0", Rotation="63.459,0.017,0.015", Scale="1,1,1") {
    rotationMode=BMM_Additive
    rotationSpace=BCS_BoneSpace
    scaleMode=BMM_Replace
    scaleSpace=BCS_BoneSpace
    translationMode=BMM_Additive
    translationSpace=BCS_BoneSpace
    alphaBoolBlend {blendOption=Linear}
    boneToModify {boneName=base_back_hair_l}
    rotation=<rot 63.4589 0.017 0.015>
   }
   n4 = "Transform (Modify) Bone - Bone: base_back_hair_r"(ComponentPose=@n3, Alpha=1, Translation="0,0,0", Rotation="55.221,-0.000,0.000", Scale="1,1,1") {
    rotationMode=BMM_Additive
    rotationSpace=BCS_BoneSpace
    scaleMode=BMM_Replace
    scaleSpace=BCS_BoneSpace
    translationMode=BMM_Additive
    translationSpace=BCS_BoneSpace
    alphaBoolBlend {blendOption=Linear}
    boneToModify {boneName=base_back_hair_r}
    rotation {pitch=55.221}
   }
   n5 = "Transform (Modify) Bone - Bone: base_side_hair_r_b"(ComponentPose=@n4, Alpha=1, Translation="0,0,0", Rotation="47.943,78.320,39.153", Scale="1,1,1") {
    rotationMode=BMM_Additive
    rotationSpace=BCS_BoneSpace
    scaleMode=BMM_Replace
    scaleSpace=BCS_BoneSpace
    translationMode=BMM_Additive
    translationSpace=BCS_BoneSpace
    alphaBoolBlend {blendOption=Linear}
    boneToModify {boneName=base_side_hair_r_b}
    rotation=<rot 47.943 78.32 39.153>
   }
   n6 = "Transform (Modify) Bone - Bone: base_side_hair_l_b"(ComponentPose=@n5, Alpha=1, Translation="0,0,0", Rotation="45.605,-49.975,-20.913", Scale="1,1,1") {
    rotationMode=BMM_Additive
    rotationSpace=BCS_BoneSpace
    scaleMode=BMM_Replace
    scaleSpace=BCS_BoneSpace
    translationMode=BMM_Additive
    translationSpace=BCS_BoneSpace
    alphaBoolBlend {blendOption=Linear}
    boneToModify {boneName=base_side_hair_l_b}
    rotation=<rot 45.605 -49.975 -20.913>
   }
   n7 = "Transform (Modify) Bone - Bone: base_side_hair_l_c"(ComponentPose=@n6, Alpha=1, Translation="0,0,0", Rotation="45.738,-68.913,-31.131", Scale="1,1,1") {
    rotationMode=BMM_Additive
    rotationSpace=BCS_BoneSpace
    scaleMode=BMM_Replace
    scaleSpace=BCS_BoneSpace
    translationMode=BMM_Additive
    translationSpace=BCS_BoneSpace
    alphaBoolBlend {blendOption=Linear}
    boneToModify {boneName=base_side_hair_l_c}
    rotation=<rot 45.738 -68.913 -31.131>
   }
   n8 = "Transform (Modify) Bone - Bone: base_side_hair_r_a"(ComponentPose=@n7, Alpha=1, Translation="-1.367,-0.275,3.071", Rotation="-4.762,83.689,2.816", Scale="1,1,1") {
    rotationMode=BMM_Additive
    rotationSpace=BCS_BoneSpace
    scaleMode=BMM_Replace
    scaleSpace=BCS_BoneSpace
    translationMode=BMM_Additive
    translationSpace=BCS_BoneSpace
    alphaBoolBlend {blendOption=Linear}
    boneToModify {boneName=base_side_hair_r_a}
    rotation=<rot -4.762 83.689 2.816>
    translation=<-1.367 -0.275 3.071>
   }
   n9 = "Transform (Modify) Bone - Bone: base_back_mid_hair"(ComponentPose=@n8, Alpha=1, Translation="0,0,0", Rotation="55.029,0.000,-0.000", Scale="1,1,1") {
    rotationMode=BMM_Additive
    rotationSpace=BCS_BoneSpace
    translationMode=BMM_Additive
    translationSpace=BCS_BoneSpace
    alphaBoolBlend {blendOption=Linear}
    boneToModify {boneName=base_back_mid_hair}
    rotation=<rot 55.029 0 0>
   }
   n10 = "Transform (Modify) Bone - Bone: base_back_mid_hair_r"(ComponentPose=@n9, Alpha=1, Translation="0,0,0", Rotation="59.468,-179.986,-180.000", Scale="1,1,1") {
    rotationMode=BMM_Additive
    rotationSpace=BCS_BoneSpace
    translationMode=BMM_Additive
    translationSpace=BCS_BoneSpace
    alphaBoolBlend {blendOption=Linear}
    boneToModify {boneName=base_back_mid_hair_r}
    rotation=<rot 59.468 -179.986 -180>
   }
   n11 = "Transform (Modify) Bone - Bone: base_side_mid_hair_l"(ComponentPose=@n10, Alpha=1, Translation="0,0,0", Rotation="49.342,0.000,-0.000", Scale="1,1,1") {
    rotationMode=BMM_Additive
    rotationSpace=BCS_BoneSpace
    translationMode=BMM_Additive
    translationSpace=BCS_BoneSpace
    alphaBoolBlend {blendOption=Linear}
    boneToModify {boneName=base_side_mid_hair_l}
    rotation=<rot 49.342 0 0>
   }
   n12 = "Transform (Modify) Bone - Bone: base_side_lower_hair_l_a"(ComponentPose=@n11, Alpha=1, Translation="0,0,0", Rotation="-31.627,-41.758,17.605", Scale="1,1,1") {
    rotationMode=BMM_Additive
    rotationSpace=BCS_BoneSpace
    translationMode=BMM_Additive
    translationSpace=BCS_BoneSpace
    alphaBoolBlend {blendOption=Linear}
    boneToModify {boneName=base_side_lower_hair_l_a}
    rotation=<rot -31.627 -41.7579 17.605>
   }
   n13 = "Transform (Modify) Bone - Bone: base_side_lower_hair_r_a"(ComponentPose=@n12, Alpha=1, Translation="0,0,0", Rotation="-27.015,-0.107,0.028", Scale="1,1,1") {
    rotationMode=BMM_Additive
    rotationSpace=BCS_BoneSpace
    translationMode=BMM_Additive
    translationSpace=BCS_BoneSpace
    alphaBoolBlend {blendOption=Linear}
    boneToModify {boneName=base_side_lower_hair_r_a}
    rotation=<rot -27.015 -0.107 0.028>
   }
   n14 = "Transform (Modify) Bone - Bone: base_side_lower_hair_l_b"(ComponentPose=@n13, Alpha=1, Translation="-6.346,-0.000,8.455", Rotation="88.717,0.010,0.010", Scale="1,1,1") {
    rotationMode=BMM_Additive
    rotationSpace=BCS_BoneSpace
    translationMode=BMM_Additive
    translationSpace=BCS_BoneSpace
    alphaBoolBlend {blendOption=Linear}
    boneToModify {boneName=base_side_lower_hair_l_b}
    rotation=<rot 88.717 0.0098 0.0098>
    translation {x=-6.346, z=8.455}
   }
   n15 = "Transform (Modify) Bone - Bone: base_back_lower_hair"(ComponentPose=@n14, Alpha=1, Translation="-8.320,-0.000,6.580", Rotation="10.265,-179.995,-179.998", Scale="1,1,1") {
    rotationMode=BMM_Additive
    rotationSpace=BCS_BoneSpace
    translationMode=BMM_Additive
    translationSpace=BCS_BoneSpace
    alphaBoolBlend {blendOption=Linear}
    boneToModify {boneName=base_back_lower_hair}
    rotation=<rot 10.265 -179.995 -179.998>
    translation {x=-8.32, z=6.58}
   }
   n16 = "Transform (Modify) Bone - Bone: base_side_lower_hair_r_b"(ComponentPose=@n15, Alpha=1, Translation="-3.009,-0.000,1.009", Rotation="52.146,-179.894,-179.905", Scale="1,1,1") {
    rotationMode=BMM_Additive
    rotationSpace=BCS_BoneSpace
    translationMode=BMM_Additive
    translationSpace=BCS_BoneSpace
    alphaBoolBlend {blendOption=Linear}
    boneToModify {boneName=base_side_lower_hair_r_b}
    rotation=<rot 52.1461 -179.8941 -179.9051>
    translation {x=-3.009, z=1.009}
   }
   n17 = RigidBody(ComponentPose=@n16, OverrideWorldGravity="0,0,0") {
    bClampLinearTranslationLimitToRefPose=true
    bOverrideWorldGravity=true
    overridePhysicsAsset=&M_MED_RevoltCrush_Mask_FaceAcc_Physics
    simulationSpace=BaseBoneSpace
    alphaBoolBlend {blendOption=Linear}
    baseBoneRef {boneName=spine_05}
    simSpaceSettings {
     dampingAlpha=0
     maxAngularAcceleration=16
     maxAngularVelocity=8
     maxLinearAcceleration=800
     maxLinearVelocity=600
     worldAlpha=1
     externalLinearDragV=<2 0.2 0.2>
    }
   }
   n18 = Get Scale vector() -> (Scale vector:Vector)
   n19 = "Transform (Modify) Bone - Bone: base_back_hair"(ComponentPose=@n17, Alpha=1, Translation="0,0,0", Rotation="0,0,0", Scale=@n18) {
    rotationMode=BMM_Additive
    rotationSpace=BCS_BoneSpace
    scaleMode=BMM_Replace
    scaleSpace=BCS_BoneSpace
    translationMode=BMM_Additive
    translationSpace=BCS_BoneSpace
    alphaBoolBlend {blendOption=Linear}
    boneToModify {boneName=base_back_hair}
   }
   n20 = "Transform (Modify) Bone - Bone: base_back_hair_l"(ComponentPose=@n19, Alpha=1, Translation="0,0,0", Rotation="0,0,0", Scale=@n18) {
    rotationMode=BMM_Additive
    rotationSpace=BCS_BoneSpace
    scaleMode=BMM_Replace
    scaleSpace=BCS_BoneSpace
    translationMode=BMM_Additive
    translationSpace=BCS_BoneSpace
    alphaBoolBlend {blendOption=Linear}
    boneToModify {boneName=base_back_hair_l}
   }
   n21 = "Transform (Modify) Bone - Bone: base_back_hair_r"(ComponentPose=@n20, Alpha=1, Translation="0,0,0", Rotation="0,0,0", Scale=@n18) {
    rotationMode=BMM_Additive
    rotationSpace=BCS_BoneSpace
    scaleMode=BMM_Replace
    scaleSpace=BCS_BoneSpace
    translationMode=BMM_Additive
    translationSpace=BCS_BoneSpace
    alphaBoolBlend {blendOption=Linear}
    boneToModify {boneName=base_back_hair_r}
   }
   n22 = "Transform (Modify) Bone - Bone: base_side_hair_r_b"(ComponentPose=@n21, Alpha=1, Translation="0,0,0", Rotation="0,0,0", Scale=@n18) {
    rotationMode=BMM_Additive
    rotationSpace=BCS_BoneSpace
    scaleMode=BMM_Replace
    scaleSpace=BCS_BoneSpace
    translationMode=BMM_Additive
    translationSpace=BCS_BoneSpace
    alphaBoolBlend {blendOption=Linear}
    boneToModify {boneName=base_side_hair_r_b}
   }
   n23 = "Transform (Modify) Bone - Bone: base_side_hair_l_b"(ComponentPose=@n22, Alpha=1, Translation="0,0,0", Rotation="0,0,0", Scale=@n18) {
    rotationMode=BMM_Additive
    rotationSpace=BCS_BoneSpace
    scaleMode=BMM_Replace
    scaleSpace=BCS_BoneSpace
    translationMode=BMM_Additive
    translationSpace=BCS_BoneSpace
    alphaBoolBlend {blendOption=Linear}
    boneToModify {boneName=base_side_hair_l_b}
   }
   n24 = "Transform (Modify) Bone - Bone: base_side_hair_l_c"(ComponentPose=@n23, Alpha=1, Translation="0,0,0", Rotation="0,0,0", Scale=@n18) {
    rotationMode=BMM_Additive
    rotationSpace=BCS_BoneSpace
    scaleMode=BMM_Replace
    scaleSpace=BCS_BoneSpace
    translationMode=BMM_Additive
    translationSpace=BCS_BoneSpace
    alphaBoolBlend {blendOption=Linear}
    boneToModify {boneName=base_side_hair_l_c}
   }
   n25 = "Transform (Modify) Bone - Bone: base_side_hair_r_a"(ComponentPose=@n24, Alpha=1, Translation="0,0,0", Rotation="0,0,0", Scale=@n18) {
    rotationMode=BMM_Additive
    rotationSpace=BCS_BoneSpace
    scaleMode=BMM_Replace
    scaleSpace=BCS_BoneSpace
    translationMode=BMM_Additive
    translationSpace=BCS_BoneSpace
    alphaBoolBlend {blendOption=Linear}
    boneToModify {boneName=base_side_hair_r_a}
   }
   n26 = "Transform (Modify) Bone - Bone: base_back_mid_hair"(ComponentPose=@n25, Alpha=1, Translation="0,0,0", Rotation="0,0,0", Scale=@n18) {
    rotationMode=BMM_Additive
    rotationSpace=BCS_BoneSpace
    scaleMode=BMM_Replace
    scaleSpace=BCS_BoneSpace
    translationMode=BMM_Additive
    translationSpace=BCS_BoneSpace
    alphaBoolBlend {blendOption=Linear}
    boneToModify {boneName=base_back_mid_hair}
   }
   n27 = "Transform (Modify) Bone - Bone: base_back_mid_hair_r"(ComponentPose=@n26, Alpha=1, Translation="0,0,0", Rotation="0,0,0", Scale=@n18) {
    rotationMode=BMM_Additive
    rotationSpace=BCS_BoneSpace
    scaleMode=BMM_Replace
    scaleSpace=BCS_BoneSpace
    translationMode=BMM_Additive
    translationSpace=BCS_BoneSpace
    alphaBoolBlend {blendOption=Linear}
    boneToModify {boneName=base_back_mid_hair_r}
   }
   n28 = "Transform (Modify) Bone - Bone: base_side_mid_hair_l"(ComponentPose=@n27, Alpha=1, Translation="0,0,0", Rotation="0,0,0", Scale=@n18) {
    rotationMode=BMM_Additive
    rotationSpace=BCS_BoneSpace
    scaleMode=BMM_Replace
    scaleSpace=BCS_BoneSpace
    translationMode=BMM_Additive
    translationSpace=BCS_BoneSpace
    alphaBoolBlend {blendOption=Linear}
    boneToModify {boneName=base_side_mid_hair_l}
   }
   n29 = "Transform (Modify) Bone - Bone: base_side_lower_hair_l_b"(ComponentPose=@n28, Alpha=1, Translation="0,0,0", Rotation="0,0,0.01", Scale=@n18) {
    rotationMode=BMM_Additive
    rotationSpace=BCS_BoneSpace
    scaleMode=BMM_Replace
    scaleSpace=BCS_BoneSpace
    translationMode=BMM_Additive
    translationSpace=BCS_BoneSpace
    alphaBoolBlend {blendOption=Linear}
    boneToModify {boneName=base_side_lower_hair_l_b}
    rotation {roll=0.01}
   }
   n30 = "Transform (Modify) Bone - Bone: base_back_lower_hair"(ComponentPose=@n29, Alpha=1, Translation="0,0,0", Rotation="0,0,0", Scale=@n18) {
    rotationMode=BMM_Additive
    rotationSpace=BCS_BoneSpace
    scaleMode=BMM_Replace
    scaleSpace=BCS_BoneSpace
    translationMode=BMM_Additive
    translationSpace=BCS_BoneSpace
    alphaBoolBlend {blendOption=Linear}
    boneToModify {boneName=base_back_lower_hair}
   }
   n31 = "Transform (Modify) Bone - Bone: base_side_lower_hair_r_b"(ComponentPose=@n30, Alpha=1, Translation="0,0,0", Rotation="0,0,0", Scale=@n18) {
    rotationMode=BMM_Additive
    rotationSpace=BCS_BoneSpace
    scaleMode=BMM_Replace
    scaleSpace=BCS_BoneSpace
    translationMode=BMM_Additive
    translationSpace=BCS_BoneSpace
    alphaBoolBlend {blendOption=Linear}
    boneToModify {boneName=base_side_lower_hair_r_b}
   }
   Output Pose(Result=Component To Local(ComponentPose=@n31)) {layerGroup=None}
  ]}
 EventGraph {nodes [
   Event BlueprintUpdateAnimation()
   > n1 = Sequence()
   > n2 = Gravity Override State Function DEPRECATED(CharacterPartAnimInstance=Self-Reference(), GameTimeInSeconds=GetGameTimeInSeconds(), GravityOverrideParams=Get Gravity Override Params())
   > Set Gravity Override(Gravity Override=@n2)
   > Set Scale vector(Scale vector:Vector="0.5,1,1") -> (Output_Get:Vector)
   > Set LOD2(LOD2:bool="Equal (Integer)"(A=Get CurrentLOD(), B=2)) -> (Output_Get:bool)
   n3 = GetOwnerSkeletalMeshComponent()
   Set Head rot(
    execute=@n1.then_1
    Head rot:double=MapRangeClamped(
     Value=GetDeltaTransformFromRefPose!(self=@n3, BoneName=head, BaseName=spine_05) -> (
      ReturnValue_Location:Vector
      ReturnValue_Rotation_Roll:float
      ReturnValue_Rotation_Pitch:float
      ReturnValue_Rotation_Yaw:float
      ReturnValue_Scale:Vector
     ).ReturnValue_Rotation_Yaw
     InRangeA=-30
     InRangeB=-100
    )
   ) -> (Output_Get:double)
  ]}
]
variables [
 Gravity Override {type=Vector}
 Gravity Override Params {type=GravityOverrideParamsStruct}
 Head rot {type=number}
 LOD2 {type=boolean}
 Scale vector {type=Vector}
]
